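# Animation Blueprint for the VibrantShell backpack cosmetic, adding secondary physics motion to its dangling chain, card frame, star, zipper and safety pin bones.
# The input pose runs through six sprung, angle-limited AnimDynamics nodes simulating root-relative with reduced gravity.
# Additive Transform Bone nodes then offset the pack, chain and card frame before a final RigidBody pass based on the pack bone with overridden gravity.
# The event graph holds only unconnected default nodes.
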AnimBlueprint /BRCosmetics/Accessories/FORT_Backpacks/Backpack_F_MED_VibrantShell/Meshes/F_MED_VibrantShell_Pack_AnimBP
target_skeleton=&/F_MED_VibrantShell_Pack_Skeleton
graphs [
 AnimGraph {nodes [
   n1 = Anim Dynamics - Bone: dyn_chain_01(ComponentPose=Local To Component(LocalPose=Input Pose()), Alpha=1, ExternalForce="0,0,0") {
    angularDampingOverride=0.999
    angularSpringConstant=12000
    bAngularSpring=true
    bOverrideAngularDamping=true
    bUseGravityOverride=true
    gravityScale=0.35
    simulationSpace=RootRelative
    boundBone {boneName=dyn_chain_01}
    componentLinearAccScale=<1 1 1>
    componentLinearVelScale=<1 1 1>
    gravityOverride {z=-980}
    physicsBodyDefinitions [{
      boundBone {boneName=dyn_chain_01}
      boxExtents=<30 15 55>
      constraintSetup {angularLimitsMax=<0 55 2>, angularLimitsMin=<0 -15 -2>, angularTarget=<1 0 0>}
      localJointOffset=<1 0 0>
     }]
    relativeSpaceBone {boneName=pack}
    simSpaceSettings {maxAngularAcceleration=35, maxAngularVelocity=7, simSpaceAngularAlpha=1}
   }
   n2 = Anim Dynamics - Bone: dyn_chain_02(ComponentPose=@n1, Alpha=1, ExternalForce="0,0,0") {
    angularDampingOverride=0.999
    angularSpringConstant=10000
    bAngularSpring=true
    bOverrideAngularDamping=true
    bUseGravityOverride=true
    bUsePlanarLimit=false
    gravityScale=0.35
    simulationSpace=RootRelative
    boundBone {boneName=dyn_chain_02}
    componentLinearAccScale=<1 1 1>
    componentLinearVelScale=<1 1 1>
    gravityOverride {z=-980}
    physicsBodyDefinitions [{
      sphereCollisionRadius=8
      boundBone {boneName=dyn_chain_02}
      boxExtents=<25 15 30>
      constraintSetup {angularLimitsMax=<5 90 5>, angularLimitsMin=<-5 -10 -5>, angularTarget=<1 0 0>}
      localJointOffset=<2 0 0>
     }]
    relativeSpaceBone {boneName=dyn_chain_01}
    simSpaceSettings {maxAngularAcceleration=35, maxAngularVelocity=7, simSpaceAngularAlpha=1}
   }
   n3 = Anim Dynamics - Bone: dyn_card_frame(ComponentPose=@n2, Alpha=1, ExternalForce="0,0,0") {
    angularDampingOverride=1
    angularSpringConstant=20000
    bAngularSpring=true
    bOverrideAngularDamping=true
    bUseGravityOverride=true
    gravityScale=0.35
    simulationSpace=RootRelative
    boundBone {boneName=dyn_card_frame}
    componentLinearAccScale=<1 1 1>
    componentLinearVelScale=<1 1 1>
    gravityOverride {z=-980}
    physicsBodyDefinitions [{
      boundBone {boneName=dyn_card_frame}
      boxExtents=<30 35 20>
      constraintSetup {angularLimitsMax=<15 45 15>, angularLimitsMin=<-15 -15 -35>, angularTarget=<1 0 0>}
      localJointOffset=<1 0 -1>
     }]
    relativeSpaceBone {boneName=dyn_chain_01}
    simSpaceSettings {maxAngularAcceleration=30, maxAngularVelocity=6, simSpaceAngularAlpha=1}
   }
   n4 = Anim Dynamics - Bone: dyn_star(ComponentPose=@n3, Alpha=1, ExternalForce="0,0,0") {
    angularDampingOverride=0.9999
    angularSpringConstant=12000
    bAngularSpring=true
    bOverrideAngularDamping=true
    bUseGravityOverride=true
    gravityScale=0.35
    simulationSpace=RootRelative
    boundBone {boneName=dyn_star}
    componentLinearAccScale=<1 1 1>
    componentLinearVelScale=<1 1 1>
    gravityOverride {z=-980}
    physicsBodyDefinitions [{
      boundBone {boneName=dyn_star}
      boxExtents=<20 30 20>
      constraintSetup {angularLimitsMax=<0 35 35>, angularLimitsMin=<0 -35 -35>, angularTarget=<1 0 0>}
      localJointOffset=<1 0 -2>
     }]
    simSpaceSettings {maxAngularAcceleration=30, maxAngularVelocity=6, simSpaceAngularAlpha=1}
   }
   n5 = Anim Dynamics - Bone: dyn_zipper(ComponentPose=@n4, Alpha=1) {
    angularDampingOverride=1
    angularSpringConstant=6000
    bAngularSpring=true
    bOverrideAngularDamping=true
    bUseGravityOverride=true
    gravityScale=0.35
    simulationSpace=RootRelative
    boundBone {boneName=dyn_zipper}
    componentLinearAccScale=<1 1 1>
    componentLinearVelScale=<1 1 1>
    gravityOverride {z=-980}
    physicsBodyDefinitions [{
      boundBone {boneName=dyn_zipper}
      boxExtents=<15 15 10>
      constraintSetup {angularLimitsMax=<0 90 5>, angularLimitsMin=<0 0 -5>, angularTarget=<1 0 0>}
      localJointOffset=<3 0 0>
     }]
    relativeSpaceBone {boneName=pack}
    simSpaceSettings {maxAngularAcceleration=30, maxAngularVelocity=6, simSpaceAngularAlpha=1}
   }
   n6 = Anim Dynamics - Bone: dyn_safetypin(ComponentPose=@n5, Alpha=1, ExternalForce="0,0,0") {
    angularDampingOverride=1
    angularSpringConstant=12000
    bAngularSpring=true
    bOverrideAngularDamping=true
    bUseGravityOverride=true
    gravityScale=0.35
    simulationSpace=RootRelative
    boundBone {boneName=dyn_safetypin}
    componentLinearAccScale=<1 1 1>
    componentLinearVelScale=<1 1 1>
    gravityOverride {z=-980}
    physicsBodyDefinitions [{
      boundBone {boneName=dyn_safetypin}
      boxExtents=<20 26 31>
      constraintSetup {angularLimitsMax=<0 15 -2>, angularLimitsMin=<0 -15 -2>, angularTarget=<1 0 0>}
      localJointOffset=<3 0 0>
     }]
    simSpaceSettings {maxAngularAcceleration=30, maxAngularVelocity=6, simSpaceAngularAlpha=1}
   }
   n7 = "Transform (Modify) Bone - Bone: pack"!(
    ComponentPose="Transform (Modify) Bone - Bone: pack"!(ComponentPose=@n6, Alpha:float, Translation="0,-1,0", Rotation="0,0,0", Scale) -> (Pose) {translationMode=BMM_Additive, boneToModify {boneName=pack}, translation {y=-1}}
    bAlphaBoolEnabled:bool
    Translation="0,-1.75,0"
    Rotation="0,0,3"
    Scale
   ) -> (Pose) {
    alphaInputType=Bool
    rotationMode=BMM_Additive
    translationMode=BMM_Additive
    alphaBoolBlend {blendInTime=1, blendOutTime=0.25}
    boneToModify {boneName=pack}
    rotation {roll=3}
    translation {y=-1.75}
   }
   n8 = "Transform (Modify) Bone - Bone: dyn_chain_01"(ComponentPose=@n7, Alpha=1, Translation="0,0,0", Rotation="0,0,35", Scale="1,1,1") {
    rotationMode=BMM_Additive
    translationMode=BMM_Additive
    alphaScaleBiasClamp {bInterpResult=true}
    boneToModify {boneName=dyn_chain_01}
    rotation {roll=35}
   }
   n9 = "Transform (Modify) Bone - Bone: dyn_chain_02"(ComponentPose=@n8, Alpha=1, Translation="0,0,0", Rotation="0,0,30", Scale="1,1,1") {
    rotationMode=BMM_Additive
    translationMode=BMM_Additive
    alphaScaleBiasClamp {bInterpResult=true}
    boneToModify {boneName=dyn_chain_02}
    rotation {roll=30}
   }
   n10 = "Transform (Modify) Bone - Bone: dyn_card_frame"(ComponentPose=@n9, Alpha=1, Translation="0,-0.5,0", Rotation="0,15,10", Scale="1,1,1") {
    rotationMode=BMM_Additive
    translationMode=BMM_Additive
    alphaScaleBiasClamp {bInterpResult=true}
    boneToModify {boneName=dyn_card_frame}
    rotation {roll=10, yaw=15}
    translation {y=-0.5}
   }
   n11 = "Transform (Modify) Bone - Bone: dyn_chain_01"(ComponentPose=@n10, Alpha=1, Translation="0,-1,0", Rotation="0,0,0", Scale="1,1,1") {
    rotationMode=BMM_Additive
    translationMode=BMM_Additive
    alphaScaleBiasClamp {bInterpResult=true}
    boneToModify {boneName=dyn_chain_01}
    translation {y=-1}
   }
   n12 = "Transform (Modify) Bone - Bone: pack"(ComponentPose=@n11, Translation="0,-1,0", Rotation="0,0,0") {
    rotationMode=BMM_Additive
    translationMode=BMM_Additive
    alphaScaleBiasClamp {bInterpResult=true}
    boneToModify {boneName=pack}
    translation {y=-1}
   }
   n13 = RigidBody(ComponentPose=@n12) {
    bOverrideWorldGravity=true
    baseBoneRef {boneName=pack}
    overrideWorldGravity {z=-1100}
    simSpaceSettings {dampingAlpha=0, maxAngularAcceleration=40, maxAngularVelocity=8, maxLinearAcceleration=2400, maxLinearVelocity=600, worldAlpha=0.45}
   }
   Output Pose(Result=Component To Local(ComponentPose=@n13)) {layerGroup=None}
  ]}
 EventGraph {nodes [
   TryGetPawnOwner()
   Event BlueprintUpdateAnimation()
  ]}
]
variables []
