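# Physics asset for the VibrantShell backpack cosmetic, setting up dynamic motion for its attached trinkets: camera, earbud case, glasses, plush and scrunchie.
# Each bone has a single capsule body, with a tiny one on the pack.
# The goods body joins to the pack and carries camera, earbud case, glasses and plush, while the scrunchie chains off the camera.
# Narrow swing and twist limits, tuned per joint, keep the motion restrained.
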PhysicsAsset /BRCosmetics/Accessories/FORT_Backpacks/Backpack_F_MED_VibrantShell/Meshes/F_MED_VibrantShell_Pack_Physics
bodies {
 dyn_camera [{
   shapeType=Capsule
   length=6.9092
   radius=2.376
   shapeName=dyn_camera_capsule
   center=<-0.0308 -0.0052 0.099>
   rotation=<rot -0.1364 -169.1312 -88.3217>
  }]
 dyn_earbudCase [{shapeType=Capsule, length=3.357, radius=2.83, shapeName=dyn_earbudCase_capsule, rotation=<rot -0.8506 2.5482 108.4459>}]
 dyn_glasses [{
   shapeType=Capsule
   length=6.7181
   radius=1.5528
   shapeName=dyn_glasses_capsule
   center=<0.0338 0.1841 -0.4285>
   rotation=<rot -0.0987 -168.8541 -89.3635>
  }]
 dyn_goods [{shapeType=Capsule, length=0.505, radius=0.505, shapeName=dyn_goods_capsule, rotation=<rot 0 -90 0>}]
 dyn_plush [{shapeType=Capsule, length=2.105, radius=2.9566, shapeName=dyn_plush_capsule, rotation=<rot 18.0553 -179.1404 -90>}]
 dyn_scrunchie [{
   shapeType=Capsule
   length=4.2118
   radius=2.554
   shapeName=dyn_scrunchie_capsule
   center=<1.1105 -0.3806 0.8709>
   rotation=<rot 0 88.6762 -83.7934>
  }]
 pack [{shapeType=Capsule, length=0.05, radius=0.05}]
}
constraints [
 {bone1Name=dyn_goods, bone2Name=pack, swing1LimitDegrees=12, swing2LimitDegrees=2, twistLimitDegrees=0}
 {bone1Name=dyn_camera, bone2Name=dyn_goods, swing1LimitDegrees=6, swing2LimitDegrees=6, twistLimitDegrees=0}
 {bone1Name=dyn_scrunchie, bone2Name=dyn_camera, swing1LimitDegrees=10, swing2LimitDegrees=5, twistLimitDegrees=7}
 {bone1Name=dyn_earbudCase, bone2Name=dyn_goods, swing1LimitDegrees=25, swing2LimitDegrees=8, twistLimitDegrees=2}
 {bone1Name=dyn_glasses, bone2Name=dyn_goods, swing1LimitDegrees=12.96, swing2LimitDegrees=10.08, twistLimitDegrees=4}
 {bone1Name=dyn_plush, bone2Name=dyn_goods, swing1LimitDegrees=14, swing2LimitDegrees=15, twistLimitDegrees=5}
]
